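# Niagara weapon trail for the PsychicAphid pickaxe, drawing ribbons from skeletal mesh bones or sockets through the shared AnimTrail modules.
# Trail and the much narrower Trail002 shift from green to blue while fading over short lifetimes.
# Dark001 adds a dark purple ribbon after a brief delay, also fading over its loop age, with loop duration taken from System.FXLifeLength.
# All three scale opacity by time of day, dimming most at night.
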NiagaraSystem /BRCosmetics/Weapons/FORT_Melee/Pickaxe_PsychicAphid/FX/NS_Pickaxe_PsychicAphid_Trail
dependencies {
 usedDataInterfaces [&NiagaraDataInterfaceColorCurve, &NiagaraDataInterfaceCurve]
 usedDynamicInputs []
 usedModules [&SystemState]
}
emitters {
 Dark001 {
  emitterUpdateScript [
   &EmitterState(Life Cycle Mode=::NewEnumerator1, Loop Behavior=::NewEnumerator1, Loop Delay=0.1, Loop Duration=$System.FXLifeLength, UseLoopDelay=true)
   &FN_SpawnAnimTrail_v2(AnimTrailWidthMode=::FromSecond, Number of Anim Trails:NiagaraInt32=&Multiply_Int(A=1, B=1))
  ]
  particleSpawnScript [
   /Game/Effects/Niagara/Modules/AnimTrail/FN_InitializeAnimTrail/FN_InitializeAnimTrailBoneIndecies(Second Bone or Socket Index=1)
   /Game/Effects/Niagara/Modules/AnimTrail/FN_InitializeAnimTrail/FN_InitializeAnimTrail(AnimTrailWidth=1, Color=<rgba 0.1135 0.07 0.125 1.5>, Lifetime=0.25)
  ]
  particleUpdateScript [
   &FN_UpdateAnimTrail_v2()
   &ScaleColor(
    Scale Alpha=&FloatFromCurve(
     FloatCurve=DataInterface() {propertyValues {Curve {keys [
         {arriveTangent=-0.2226, interpMode=RCIM_Cubic, leaveTangent=-0.2226, tangentMode=RCTM_User, time=0.1, value=1}
         {arriveTangent=-0.111, interpMode=RCIM_Cubic, leaveTangent=-0.111, tangentMode=RCTM_User, time=0.3, value=0.25}
         {arriveTangent=-0.0236, tangentMode=RCTM_User, time=1, value=0}
        ]}}}
     Scale Curve=&Multiply_Float(A=5, B=&FNTimeOfDayScalar(Night Scale=0.3, Morning Scale=0.4, Evening Scale=0.4))
    )
    ScaleA=true
    ScaleRGB=false
   )
   &SpriteFacingAndAlignment(Sprite Facing=$Particles.RibbonFacing, WriteAlignment=false, WriteFacing=true)
   &ScaleColor(
    Scale Alpha=&FloatFromCurve(
     FloatCurve=DataInterface() {propertyValues {Curve {keys [
         {tangentMode=RCTM_User, time=0, value=1}
         {arriveTangent=-0.0405, interpMode=RCIM_Cubic, leaveTangent=-0.0405, tangentMode=RCTM_User, time=0.3, value=1}
         {arriveTangent=-3.8778, interpMode=RCIM_Cubic, leaveTangent=-3.8778, tangentMode=RCTM_User, time=0.7, value=0.075}
        ]}}}
     CurveIndex=$Emitter.NormalizedLoopAge
    )
    ScaleA=true
    ScaleRGB=true
   ) {module=ScaleColor001}
  ]
  rendererClasses [&NiagaraRibbonRendererProperties]
  renderers [{rendererClass=&NiagaraRibbonRendererProperties}]
 }
 Trail {
  emitterUpdateScript [
   &EmitterState()
   &FN_SpawnAnimTrail_v2()
  ]
  particleSpawnScript [
   &FN_InitializeAnimTrailBoneIndecies(Second Bone or Socket Index=1)
   &FN_InitializeAnimTrail(AnimTrailWidth=1, Color=<rgba 1 1 1 0.8>, Lifetime=0.17)
  ]
  particleUpdateScript [
   &FN_UpdateAnimTrail_v2()
   &ScaleColor(
    Scale Alpha=&FloatFromCurve(
     FloatCurve=DataInterface() {propertyValues {Curve {keys [{arriveTangent=-2.1531, interpMode=RCIM_Cubic, leaveTangent=-2.1531, tangentMode=RCTM_User, time=0, value=1}, <at 1 0 RCIM_Cubic>]}}}
     Scale Curve=&FNTimeOfDayScalar(Night Scale=0.3, Morning Scale=0.7, Evening Scale=0.7)
    )
    Scale RGB=&Multiply_Vector(
     A=<1 1 1>
     B=&MakeVectorFromLinearColorRGB(Color=&ColorFromCurve(ColorCurve=DataInterface() {propertyValues {
         AlphaCurve {keys [<at 0 1 RCIM_Cubic>]}
         BlueCurve {keys [<at 0 0.2>, {arriveTangent=1.0407, interpMode=RCIM_Cubic, leaveTangent=1.0407, time=0.3274, value=0.1}, <at 0.7687 1>]}
         GreenCurve {keys [<at 0 1>, {arriveTangent=-0.4488, interpMode=RCIM_Cubic, leaveTangent=-0.4488, time=0.3274, value=1}, <at 0.7687 0.655>]}
         RedCurve {keys [<at 0 0.2933>, {arriveTangent=-0.2515, interpMode=RCIM_Cubic, leaveTangent=-0.2515, time=0.3274, value=0.205}, <at 0.7687 0.1>]}
        }}))
    )
    ScaleA=true
    ScaleRGB=true
   )
  ]
  rendererClasses [&NiagaraRibbonRendererProperties]
  renderers [{rendererClass=&NiagaraRibbonRendererProperties}]
 }
 Trail002 {
  emitterUpdateScript [
   &EmitterState()
   &FN_SpawnAnimTrail_v2(AnimTrailWidthMode=::FromSecond)
  ]
  particleSpawnScript [
   &FN_InitializeAnimTrailBoneIndecies(Second Bone or Socket Index=1)
   &FN_InitializeAnimTrail(AnimTrailWidth=0.05, Color=<rgba 1 1 1 0.5>, Lifetime=0.2)
  ]
  particleUpdateScript [
   &FN_UpdateAnimTrail_v2()
   &ScaleColor(
    Scale Alpha=&FloatFromCurve(
     FloatCurve=DataInterface() {propertyValues {Curve {keys [{arriveTangent=-2.1531, interpMode=RCIM_Cubic, leaveTangent=-2.1531, tangentMode=RCTM_User, time=0, value=1}, <at 1 0 RCIM_Cubic>]}}}
     Scale Curve=&FNTimeOfDayScalar(Night Scale=0.5, Morning Scale=0.7, Evening Scale=0.7)
    )
    Scale RGB=&Multiply_Vector(
     A=&VectorFromFloat(Value=1.3)
     B=&MakeVectorFromLinearColorRGB(Color=&ColorFromCurve(ColorCurve=DataInterface() {propertyValues {
         AlphaCurve {keys [<at 0 1 RCIM_Cubic>]}
         BlueCurve {keys [<at 0 0.2>, {arriveTangent=1.0407, interpMode=RCIM_Cubic, leaveTangent=1.0407, time=0.3274, value=0.1}, <at 0.7687 1>]}
         GreenCurve {keys [<at 0 1>, {arriveTangent=-0.4488, interpMode=RCIM_Cubic, leaveTangent=-0.4488, time=0.3274, value=1}, <at 0.7687 0.655>]}
         RedCurve {keys [<at 0 0.2933>, {arriveTangent=-0.2515, interpMode=RCIM_Cubic, leaveTangent=-0.2515, time=0.3274, value=0.205}, <at 0.7687 0.1>]}
        }}))
    )
    ScaleA=true
    ScaleRGB=true
   )
  ]
  rendererClasses [&NiagaraRibbonRendererProperties]
  renderers [{rendererClass=&NiagaraRibbonRendererProperties}]
 }
}
summary {
 systemName=NS_Pickaxe_PsychicAphid_Trail
 emitters [
  {emitterName=Trail, rendererClasses [&NiagaraRibbonRendererProperties]}
  {emitterName=Dark001, rendererClasses [&NiagaraRibbonRendererProperties]}
  {emitterName=Trail002, rendererClasses [&NiagaraRibbonRendererProperties]}
 ]
 userVariables [
  User.bIsFrontEndPreview {type=NiagaraBool}
  User.RibbonIDOffset {type=NiagaraInt32}
  User.SkeletalMesh {
   type=NiagaraDataInterfaceSkeletalMesh
   defaultValue {
    struct=NiagaraExt_VariableValue_DataInterface
    value {dataInterface=&self:NiagaraDataInterfaceSkeletalMesh_0}
   }
  }
 ]
}
